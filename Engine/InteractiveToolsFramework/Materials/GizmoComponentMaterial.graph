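# Translucent unlit material for editor-tool gizmo components that draws without depth testing and undoes exposure on its color.
# GizmoColor is shaded by view-versus-normal facing, using derivative-based flat normals except in orthographic views.
# An optional switch dims pixels behind scene depth with a sparse screen-space dither.
# Another switch masks alpha against custom depth with a distance-scaled bias, so components writing to that buffer sort among themselves.
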Material /Engine/InteractiveToolsFramework/Materials/GizmoComponentMaterial
graph {
 properties {
  allowTranslucentCustomDepthWrites=true
  bDisableDepthTest=true
  blendMode=BLEND_Translucent
  pixelDepthOffsetMode=PDOM_Legacy
  shadingModel=MSM_Unlit
  usedShadingModels=Unlit
 }
 nodes [
  n1 = Constant() {r=1}
  n2 = If(A=PixelDepth().R, B=SceneDepth().R, "A > B"=NamedRerouteUsage(), "A == B"=@n1, "A < B"=@n1) {desc="If pixel is deeper than depth, dither/dim it"}
  n3 = VectorParameter() {parameterName=GizmoColor, defaultValue=<rgba 1 1 1 1>}
  n4 = Multiply(A=@n3.RGB, B=NamedRerouteUsage())
  EyeAdaptationInverse(LightValueInput=StaticSwitchParameter(True=Multiply(A=@n2, B=@n4), False=@n4) {defaultValue=true, parameterName=DimRegularOcclusion}) {desc=Undo effect of exposure}
  n5 = PixelDepth()
  n6 = Subtract(A=@n5.R, B=SceneTexture() {sceneTextureId=PPI_CustomDepth}.Color)
  n7 = Divide(A=@n5.R, B=Constant() {r=10000})
  n8 = Multiply(A=@n7, B=@n7)
  n9 = If(A=Round(None=IsOrthographic()), "A > B"=@n8, "A == B"=Divide(A=@n7) {constB=100}, "A < B"=@n8) {constB=1}
  n10 = Constant() {r=1}
  n11 = StaticSwitchParameter(
   True=If(A=Subtract(A=@n6, B=@n9), "A > B"=Constant(), "A == B"=@n10, "A < B"=@n10) {constB=0.01, desc="If this depth-customDepth > 0"}
   False=Constant() {r=1}
  ) {
   defaultValue=true
   desc="Custom depth can allow components \r\nwith this material to be properly sorted \r\nwithin their translucency priority, as \r\nlong as they all write to the buffer."
   parameterName=OccludeByCustomDepth
  }
  Multiply(A=@n3.A, B=@n11)
  n12 = CameraVectorWS()
  n13 = Round(None=IsOrthographic()) {desc="We round here because IsOrthographic produces an integer value\r\n which isn't acceptable by If nodes as input under restrictive mode."}
  n14 = WorldPosition() {desc=Flat Normal}
  n15 = Normalize(VectorInput=CrossProduct(A=DDX(Value=@n14.XYZ) {desc=Flat Normal}, B=DDY(Value=@n14.XYZ) {desc=Flat Normal}) {desc=Flat Normal}) {desc=Flat Normal}
  n16 = If(A=@n13, "A > B"=@n15, "A == B"=PixelNormalWS(), "A < B"=@n15) {constB=1, desc="DDX/DDY produces artifacts in ortho\r\nviews so use input Normal then"}
  n17 = Clamp(None=Abs(None=DotProduct(A=@n12, B=@n16)))
  NamedRerouteDeclaration(None=@n17) {name=NormalBasedColorAdjustment, nodeColor=<rgba 1 0 0.2812 1>}
  n18 = ScreenPosition()
  n19 = Constant() {desc=Sets checkerboard size, r=2}
  n20 = Floor(None=Fmod(A=Divide(A=@n18.PixelPosition, B=@n19), B=Constant() {r=2}))
  n21 = ComponentMask(None=@n20) {r=true}
  n22 = ComponentMask(None=@n20) {g=true}
  n23 = Constant()
  NamedRerouteDeclaration(None=If(
    A=Add(A=@n21, B=@n22) {desc="We create a sparse dither by making every 0,0 black.\r\nTo do a checkerboard, we would instead compare R and G"}
    "A > B"=ScalarParameter() {defaultValue=0.5, desc=value in the non black squares, parameterName=OcclusionMultiplier}
    "A == B"=@n23
    "A < B"=@n23
   )) {name=DitherDim, nodeColor=<rgba 0 0.4375 1 1>}
  Power(Base=@n17) {constExponent=1.5, desc=Use this node for a more rounded look}
 ]
}
